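# Sound cue for the Cube healing tick in Figment S05, reusing four Campsite Tent healing tick waves.
# An equally weighted Random node picks one wave and feeds a Modulator that slightly varies pitch and volume.
# The cue plays quietly under the non-critical in-game UI sound class, with its own short-range attenuation and a concurrency cap on simultaneous ticks.
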SoundCue /Figment_S05_Cube/Sounds/Cube_Healing_Tick_Cue
root=Modulator_1
nodes [
 {id=Modulator_1, children [Random_1], properties {pitchMax=1.1, pitchMin=0.9, volumeMax=1, volumeMin=0.7}}
 {id=Random_1, children [WavePlayer_0,WavePlayer_1,WavePlayer_6,WavePlayer_7], properties {weights [1,1,1,1]}}
 {id=WavePlayer_0, properties {soundWaveAssetPtr=&/Campsite_Tent_HealingTick_01}}
 {id=WavePlayer_1, properties {soundWaveAssetPtr=&/Campsite_Tent_HealingTick_02}}
 {id=WavePlayer_6, properties {soundWaveAssetPtr=&/Campsite_Tent_HealingTick_03}}
 {id=WavePlayer_7, properties {soundWaveAssetPtr=&/Campsite_Tent_HealingTick_04}}
]
properties {
 bOverrideAttenuation=true
 bOverrideConcurrency=true
 maxDistance=612
 soundClassObject=&Fort_UI_InGame_NonCritical
 volumeMultiplier=0.1
 attenuationOverrides {falloffDistance=512, stereoSpread=1500, attenuationShapeExtents {x=100}}
 concurrencyOverrides {maxCount=3}
}
